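# Material function that adds parallax to cubemap reflections by anchoring a cubemap to a capture position in the world.
# It intersects the world-space reflection ray, scaled by Radius, with a slightly enlarged sphere around CapturePosition by solving a quadratic.
# The far hit point, relative to CapturePosition, becomes the cubemap lookup vector.
# Rays that miss return black, and a second output provides a sphere hit mask from the determinant.
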MaterialFunction /Engine/Functions/Engine_MaterialFunctions01/Reflections/DistanceLimitedReflections
graph {
 properties {
  bExposeToLibrary=true
  description=Distance Limited Reflections give some parallax to cubemaps by fixing them in a point in space and performing a ray sphere intersection. Can be used to help anchor manually placed cubemaps in the world.
  libraryCategoriesText [Reflections]
 }
 nodes [
  n1 = FunctionInput() {
   bUsePreviewValueAsDefault=true
   description=The radius of the sphere used for projection.
   inputName=Radius
   inputType=FunctionInput_Scalar
   sortPriority=2
   previewValue=<100 0 0 1>
  }
  n2 = Multiply(A=@n1, B=ReflectionVectorWS()) {desc=RayDirection}
  n3 = FunctionInput() {
   bUsePreviewValueAsDefault=true
   description="The position that the cubemap was captured from, or the place you want it to be centered in the world."
   inputName=CapturePosition
   sortPriority=1
  }
  n4 = Subtract(A=WorldPosition().XYZ, B=@n3) {desc=ReceiverToSphereCenter}
  n5 = Multiply(A=DotProduct(A=@n2, B=@n4)) {constB=2, desc=QuadraticCoef.y}
  n6 = Multiply(A=@n1) {constB=1.2, desc=ProjectionSphereRadius}
  n7 = Subtract(A=DotProduct(A=@n4, B=@n4) {desc=ReceiverToSphereCenterSq}, B=Multiply(A=@n6, B=@n6) {constB=1.2, desc=SphereRadiusSquared}) {desc=QuadraticCoef.z}
  n8 = DotProduct(A=@n2, B=@n2) {desc=QuadraticCoef.x}
  n9 = Subtract(A=Multiply(A=@n5, B=@n5), B=Multiply(A=Multiply(A=@n7, B=@n8)) {constB=4}) {desc=Determinant}
  n10 = Constant()
  n11 = Constant() {desc=Epsilon, r=0}
  n12 = Multiply(A=Subtract(A=SquareRoot(None=@n9), B=@n5), B=Divide(B=Add(A=@n8, B=@n11)) {constA=0.5, desc=InvTwoA}) {desc=FarIntersection}
  n13 = Subtract(A=Add(A=Multiply(A=@n12, B=@n2), B=WorldPosition().XYZ) {desc=IntersectPosition}, B=Multiply(A=@n3))
  n14 = FunctionInput() {inputName=Cubemap, inputType=FunctionInput_TextureCube}
  FunctionOutput(None=If(A=@n9, B=@n10, "A > B"=TextureSample(UVs=@n13, Tex=@n14).RGB, "A == B"=@n10, "A < B"=@n10))
  FunctionOutput(None=Clamp(None=Ceil(None=@n9))) {
   description=Rays that do not hit the sphere will write black pixels into the output color as well as this mask output.
   outputName=Sphere Hit Mask
   sortPriority=1
  }
 ]
}
