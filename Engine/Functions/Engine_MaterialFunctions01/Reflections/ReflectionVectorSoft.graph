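# ReflectionVectorSoft is an engine material function in the Reflections category that outputs a reflection vector from a softened normal, for smoother reflections.
# It lerps the Normal input toward a constant flat normal (0,0,1) using the scalar Softness input as alpha, described as a 0 to 1 value.
# The blended normal and the camera vector feed the CustomReflectionVector function, whose Reflection Vector is the output.
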MaterialFunction /Engine/Functions/Engine_MaterialFunctions01/Reflections/ReflectionVectorSoft
graph {
 properties {description=Softens the normal for a smoother reflection, libraryCategoriesText [Misc,Reflections]}
 nodes [
  n1 = FunctionInput() {description=Normal used for the reflection to be softened, inputName=Normal, previewValue=<0 0 1 1>}
  n2 = Constant3Vector() {constant=<rgba 0 0 1 0>}
  n3 = FunctionInput() {description=value from 0 to 1, inputName=Softness, inputType=FunctionInput_Scalar, previewValue=<0.5 0 0 1>}
  FunctionOutput(None=MaterialFunctionCall!(Normal=LinearInterpolate(A=@n1, B=@n2, Alpha=@n3), CameraVector) {materialFunction=/Engine/Functions/Engine_MaterialFunctions02/Utility/CustomReflectionVector}.Reflection Vector)
 ]
}
