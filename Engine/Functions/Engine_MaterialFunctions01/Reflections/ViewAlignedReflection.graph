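# Material function, exposed to the library, that aligns a sphere-style reflection texture to the view, like a sphere map lookup.
# The reflection vector input defaults to the world-space reflection vector transformed into view space.
# Its X and negated Y are remapped by a half scale and bias into UVs, ignoring Z.
# It outputs the RGB of the input texture sampled at those UVs, plus the UVs themselves.
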MaterialFunction /Engine/Functions/Engine_MaterialFunctions01/Reflections/ViewAlignedReflection
graph {
 properties {
  bExposeToLibrary=true
  description="Alignes a Reflection texture to your view.  Reflection should be a sphere."
  libraryCategoriesText [Misc,Reflections]
 }
 nodes [
  n1 = ReflectionVectorWS()
  n2 = MaterialFunctionCall(Float3=FunctionInput(Preview=Transform(None=@n1) {transformSourceType=TRANSFORMSOURCE_World, transformType=TRANSFORM_View}) {bUsePreviewValueAsDefault=true, inputName=ReflectionVector}) {materialFunction=&BreakOutFloat3Components}
  n3 = Multiply(A=@n2.G) {constB=-1}
  n4 = Constant() {r=0.5}
  n5 = Add(A=Multiply(A=AppendVector(A=@n2.R, B=@n3), B=@n4), B=@n4)
  n6 = TextureObject() {texture=/Engine/EngineMaterials/DefaultDiffuse}
  FunctionOutput(None=TextureSample(
    UVs=@n5
    Tex=FunctionInput(Preview=@n6) {description=Reflection should look like a sphere, inputName=ReflectionTexture, inputType=FunctionInput_Texture2D}
   ) {texture=&DefaultDiffuse}.RGB) {outputName=Texture}
  FunctionOutput(None=@n5) {description=Reflection UVs, outputName=UVs, sortPriority=1}
 ]
}
